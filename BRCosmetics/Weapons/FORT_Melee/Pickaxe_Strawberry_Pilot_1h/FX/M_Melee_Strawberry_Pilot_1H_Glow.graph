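# Translucent, unlit, two-sided glow material for mesh particle effects on the one-handed Strawberry Pilot pickaxe.
# A panning noise normal, scaled by Distort, warps a panning smoke texture whose tiled UVs shift by OffsetX.
# Smoke times particle color gets a hue shift scaled along U by HueShift, then passes through emissive adjustments.
# Smoke times particle alpha is faded by a squared falloff along the second UV channel.
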Material /BRCosmetics/Weapons/FORT_Melee/Pickaxe_Strawberry_Pilot_1h/FX/M_Melee_Strawberry_Pilot_1H_Glow
graph {
 properties {
  bUsedWithMeshParticles=true
  bUsedWithNiagaraMeshParticles=true
  blendMode=BLEND_Translucent
  pixelDepthOffsetMode=PDOM_Legacy
  shadingModel=MSM_Unlit
  twoSided=true
  usedShadingModels=Unlit
 }
 nodes [
  n1 = ComponentMask(None=TextureCoordinate()) {r=true}
  n2 = DynamicParameter() {paramNames [HueShift,OffsetX,Distort,Param4]}
  n3 = TextureCoordinate() {uTiling=0.25, vTiling=1.25}
  n4 = ComponentMask(None=@n3) {g=true}
  n5 = AppendVector(A=Add(A=DynamicParameter() {paramNames [HueShift,OffsetX,Distort,Param4]}.OffsetX, B=ComponentMask(None=@n3) {r=true}), B=@n4)
  n6 = Multiply(
   A=TextureSample(UVs=Panner(Coordinate=@n5) {speedX=-0.15}) {samplerType=SAMPLERTYPE_Normal, texture=/Game/Global/Textures/Default/Noise/T_Noise_Normal_01}.RGB
   B=DynamicParameter() {paramNames [HueShift,OffsetX,Distort,Param4]}.Distort
  )
  n7 = TextureSample(UVs=Add(A=Panner(Coordinate=@n5) {speedX=-0.15}, B=ComponentMask(None=@n6) {g=true, r=true})) {texture=&T_Inky_Smoke_Tile}
  n8 = ParticleColor()
  MaterialFunctionCall(Emissive=MaterialFunctionCall(Hue Shift Percentage=Multiply(A=@n1, B=@n2.HueShift), Texture=Multiply(A=@n7.G, B=@n8.RGB)) {materialFunction=&HueShift}.Result) {materialFunction=&MF_FX_EmissiveAdjustments}
  n9 = OneMinus(None=ComponentMask(None=TextureCoordinate() {coordinateIndex=1}) {r=true})
  Multiply(A=Multiply(A=@n7.G, B=@n8.A), B=Multiply(A=@n9, B=@n9))
  Power(Base=@n4) {constExponent=4}
 ]
}
