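# Translucent, unlit, two-sided trail material for the Strawberry Pilot one-handed pickaxe, built for beam trails and Niagara ribbons.
# Particle color gets a hue shift along U, scaled by a dynamic parameter, before emissive adjustment.
# Panning smoke and plasma textures, distorted by chromatic smoke, multiply with two summed noise layers panned over squared, inverted U.
# UV-gradient and sphere masks shape and fade that pattern, which particle alpha then scales.
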Material /BRCosmetics/Weapons/FORT_Melee/Pickaxe_Strawberry_Pilot_1h/FX/M_Melee_Strawberry_Pilot_1H_Trail
graph {
 properties {
  bUsedWithBeamTrails=true
  bUsedWithNiagaraRibbons=true
  blendMode=BLEND_Translucent
  pixelDepthOffsetMode=PDOM_Legacy
  shadingModel=MSM_Unlit
  twoSided=true
  usedShadingModels=Unlit
 }
 nodes [
  n1 = TextureCoordinate()
  ComponentMask(None=@n1) {r=true}
  n2 = ComponentMask(None=TextureCoordinate()) {r=true}
  n3 = DynamicParameter()
  n4 = ParticleColor()
  MaterialFunctionCall(Emissive=MaterialFunctionCall(Hue Shift Percentage=Multiply(A=@n2, B=@n3.Param1), Texture=@n4.RGB) {materialFunction=&HueShift}.Result) {materialFunction=&MF_FX_EmissiveAdjustments}
  n5 = Panner(Coordinate=TextureCoordinate() {uTiling=0.5, vTiling=0.5}) {speedX=0.2}
  n6 = Panner(Coordinate=TextureCoordinate() {uTiling=0.5, vTiling=0.5}) {speedX=0.15}
  n7 = Multiply(
   A=ComponentMask(None=TextureSample(UVs=@n6) {texture=&T_Inky_Smoke_Tile_Chromatic}.RGB) {g=true, r=true}
   B=ScalarParameter() {defaultValue=0.5, parameterName=Distort}
  )
  n8 = Panner(Coordinate=TextureCoordinate() {uTiling=0.5, vTiling=0.5}) {speedX=0.15}
  n9 = Multiply(
   A=TextureSample(UVs=Add(A=@n5, B=@n7)) {texture=&T_Inky_Smoke_Tile}.R
   B=TextureSample(UVs=Add(A=@n8, B=@n7)) {texture=/Game/Effects/Fort_Effects/Textures/T_FX_SpringerPlasma}.B
  )
  n10 = VectorParameter() {parameterName=Tile, defaultValue=<rgba 1 0.25 0 0>}
  n11 = MaterialFunctionCall(Float2=Multiply(A=TextureCoordinate(), B=ComponentMask(None=@n10.RGB) {g=true, r=true})) {materialFunction=&BreakOutFloat2Components}
  n12 = OneMinus(None=@n11.R)
  n13 = AppendVector(A=Multiply(A=@n12, B=@n12), B=@n11.G)
  n14 = Multiply(A=Time(), B=ScalarParameter() {defaultValue=1, parameterName=PanSpeed})
  n15 = TextureSampleParameter2D(UVs=Panner(Coordinate=@n13, Time=@n14) {speedX=0.5}) {parameterName=DetailNoise, texture=&T_ShieldCaustics}
  n16 = TextureSampleParameter2D(UVs=Panner(Coordinate=@n13, Time=@n14) {speedX=2}) {parameterName=BasicNoise, texture=&T_NoisePack_01}
  n17 = MaterialFunctionCall(Float2=TextureCoordinate()) {materialFunction=&BreakOutFloat2Components}
  n18 = Power(Base=OneMinus(None=@n17.R))
  n19 = ComponentMask(None=TextureCoordinate()) {g=true}
  n20 = Multiply(
   A=OneMinus(None=@n17.G)
   B=SphereMask(A=Power(Base=@n19) {constExponent=0.35}, B=Constant() {r=0.48}) {attenuationRadius=0.5, hardnessPercent=0}
  )
  n21 = Multiply(A=Multiply(A=Add(A=@n15.R, B=@n16.R), B=Multiply(A=@n18, B=@n20))) {constB=2}
  n22 = ComponentMask(None=TextureCoordinate()) {r=true}
  n23 = OneMinus(None=Power(Base=OneMinus(None=@n22)) {constExponent=0.5})
  n24 = ComponentMask(None=@n1) {g=true}
  n25 = Multiply(A=@n24, B=OneMinus(None=@n24))
  Multiply(A=@n4.A, B=Multiply(A=Multiply(A=@n9, B=@n21), B=Multiply(A=@n23, B=@n25)))
  TextureCoordinate()
 ]
}
