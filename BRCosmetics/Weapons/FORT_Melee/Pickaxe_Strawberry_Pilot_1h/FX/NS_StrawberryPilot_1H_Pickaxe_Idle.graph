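# Niagara idle effect for the Strawberry Pilot one-handed pickaxe cosmetic, with color variants driven by User.VariantID.
# A system spawn assignment, using float-to-int and color-array inputs, supplies VariantColor to particle initialization.
# One emitter spawns particles at a steady rate with looping lifetimes, feeding both sprite and mesh renderers.
# Particle updates boost RGB, fade alpha in by curve, and drive dynamic material parameters with curve and random values.
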NiagaraSystem /BRCosmetics/Weapons/FORT_Melee/Pickaxe_Strawberry_Pilot_1h/FX/NS_StrawberryPilot_1H_Pickaxe_Idle
dependencies {
 usedDataInterfaces [&NiagaraDataInterfaceCurve, &NiagaraDataInterfaceArrayColor]
 usedDynamicInputs [&SelectLinearColorFromArray, &MakeIntFromFloat]
 usedModules [
  &SystemState
  &self:SystemSpawnScript.SystemScriptSource.SystemScriptGraph.NiagaraNodeAssignment_1.SetVariables_#1
 ]
}
emitters {NE_Empty {
  emitterUpdateScript [
   &EmitterState(
    Inactive Response=::NewEnumerator1
    Life Cycle Mode=::NewEnumerator1
    Loop Behavior=::NewEnumerator1
    Loop Duration Mode=::NewEnumerator1
    UseLoopDelay=false
   )
   &SpawnRate(SpawnRate=20, Use Spawn Probability=false)
  ]
  particleSpawnScript [&InitializeParticle(Color=$System.VariantColor, UsePositionOffset=false, Write Material Random=true) {module=InitializeParticle001}]
  particleUpdateScript [
   &ParticleState(
    Kill Particles When Lifetime Has Elapsed=false
    Let Infinitely Lived Particles Die When Emitter Deactivates=true
    Loop Particles Lifetime=true
   )
   &SolveForcesAndVelocity(Clamp Velocity=false, Limit Acceleration=false)
   &ScaleColor(
    Scale Alpha=&FloatFromCurve(
     FloatCurve=DataInterface() {propertyValues {Curve {
        defaultValue=1
        keys [
         {interpMode=RCIM_Cubic, tangentMode=RCTM_User, time=0, value=0}
         {interpMode=RCIM_Cubic, tangentMode=RCTM_User, time=0.5, value=1}
         {tangentMode=RCTM_User, time=1, value=1}
        ]
       }}}
     Scale Curve=0.15
    )
    Scale RGB=<3 3 3>
    ScaleA=true
    ScaleRGB=true
   )
   &DynamicMaterialParameters(
    Index 0 Param 1=&FloatFromCurve(FloatCurve=DataInterface() {propertyValues {Curve {defaultValue=1, keys [{tangentMode=RCTM_User, time=0, value=0}, {tangentMode=RCTM_User, time=1, value=0.15}]}}})
    Index 0 Param 3=&RandomRangeFloat(Minimum=0.05, Maximum=0.15)
    Module.Index0Float0Write=true
    Module.Index0Float1Write=true
    Module.Index0Float2Write=true
    Module.Index0Float3Write=true
   )
  ]
  rendererClasses [&NiagaraSpriteRendererProperties, &NiagaraMeshRendererProperties]
  renderers [
   {rendererClass=&NiagaraSpriteRendererProperties}
   {rendererClass=&NiagaraMeshRendererProperties, rendererIndex=1}
  ]
 }}
summary {
 systemName=NS_StrawberryPilot_1H_Pickaxe_Idle
 emitters [{emitterName=NE_Empty, rendererClasses [&NiagaraSpriteRendererProperties, &NiagaraMeshRendererProperties]}]
 userVariables [User.VariantID {type=NiagaraFloat, defaultValue=0, description=""}]
}
